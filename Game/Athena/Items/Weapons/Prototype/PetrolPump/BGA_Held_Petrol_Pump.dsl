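; Held petrol pump nozzle actor that drops itself when carried too far from its pump.
; On BeginPlay it caches the instigator pawn and starts a looping timer comparing the current weapon's position to a stored location.
; Exceeding the scaled tether range clears the timer and drops the held object; on drop, the authority detaches the ApolloGasPumpValet cable and destroys the actor.
; Interaction checks defer to the held object component, with interact time from a scalable float.
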(event EventBeginPlay
  ;[1 1 1] Once we get the pump, start a timer to check location
  (|CallParentFunction)
  (bind _asfort_pawn (Utilities|Casting|CastToFortPawn (Game|GetInstigator)))
  (Variables|Default|SetPawn _asfort_pawn)
  (bind _returnvalue (Utilities|Time|SetTimerbyEvent (EventDispatchers|CreateEvent :!event Custom|PumpDistanceCheck) 0.25 true))
  (Variables|Default|SetDistanceCheckTimer _returnvalue))

(event Custom|PumpDistanceCheck
  (bind _fortheldobject (Variables|Default|GetFortHeldObject))
  ;[1 1 1] Distance from pump check. If we are too far from pump then drop the nozzle
  (bind _pawn (Variables|Default|GetPawn)
    (:then
      (bind _currentweapon (Variables|/Script/FortniteGame.FortPawn|GetCurrentWeapon _pawn)
        (:then
          (bind _distance (Class|MissionBlueprintFunctionLibrary|DistanceBetweenTwoVectors (Transformation|GetActorLocation _currentweapon) (Variables|Default|GetLocation)))
          (if (>= _distance (* (ScalableFloat|GetValueatLevel (Variables|Default|GetPumpTetherRange)) (Variables|Default|GetTetherRangeScale)))
            (Utilities|Time|ClearandInvalidateTimerbyHandle (Variables|Default|GetDistanceCheckTimer))
            (Utilities|IsValid _fortheldobject
              (:Is%20Valid
                (HeldObjects|Behavior|DropHeldObject _fortheldobject)))))
        (:else)))
    (:else)))

(event OnHeldObjectDropped(FortHeldObject)
  ;[1 1 1] EVENT -> On Dropped - Destroy pump
  (Networking|SwitchHasAuthority
    (:Authority
      (Utilities|IsValid (Variables|Default|GetPump)
        (:Is%20Valid
          (Class|ApolloGasPumpValet|SetCableAttachedonServer (Variables|Default|GetPump) false))
        (:Is%20Not%20Valid
          _))
      (Actor|DestroyActor))
    (:Remote
      _)))

(fn BlueprintCanInteract
  (params (InteractingPawn :type Fort%20Pawn%20Object%20Reference) (InteractionBeingAttempted :type EInteractionBeingAttempted%20Enum :default "FirstInteraction") (InteractionType :type TInteractionType%20Enum :default "IT_NoInteraction"))
  (bind _returnvalue (HeldObjects|Interaction|CanInteract (Variables|Default|GetFortHeldObject) (Utilities|Casting|CastToFortPlayerPawn InteractingPawn)))
  (return _returnvalue))

(fn BlueprintGetInteractionTime
  (params (InteractingPawn :type Fort%20Pawn%20Object%20Reference) (InteractionBeingAttempted :type EInteractionBeingAttempted%20Enum :default "FirstInteraction"))
  (return (ScalableFloat|GetValueatLevel (Variables|Default|GetInteractTime)) false))

(fn ConstructionScript
  (|CallParentFunction))
