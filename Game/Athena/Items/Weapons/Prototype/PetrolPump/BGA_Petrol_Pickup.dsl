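; Gas can actor for Athena that players pick up into inventory, can be attached to a held weapon, and can be refueled.
; Damage or Curie fire lights a random server fuse, dropping any held can as a projectile.
; Expiry spawns Curie fireballs, then applies line-of-sight-checked radial damage by target type, plus vehicle impulses.
; ShouldDie returns false to preserve collision when shot from hands, leaving destruction to HideAndDestroy.
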(event EventBeginPlay
  ;[1 1 1] Begin Play
  (|CallParentFunction)
  (Variables|Interaction|SetbAllowInteract false)
  (Utilities|FlowControl|Delay 1.0)
  (Variables|Interaction|SetbAllowInteract true))

(event Interaction|EventBlueprintOnInteract
  (params (InteractingPawn :type Fort%20Pawn%20Object%20Reference) (InteractionBeingAttempted :type EInteractionBeingAttempted%20Enum :default "FirstInteraction"))
  ;[1 1 1] EVENT -> Interact
  (if (Utilities|Enum|Equal(Enum) (Fortnite|GetSubGame) "Athena")
    (bind _asfort_player_pawn (Utilities|Casting|CastToFortPlayerPawn InteractingPawn))
    (bind _output_get (Variables|Default|SetPlayerPawn _asfort_player_pawn))
    (bind _asfort_player_controller_athena (Utilities|Casting|CastToFortPlayerControllerAthena (Fortnite|GetFortPlayerControllerfromActor _output_get)))
    (Variables|Default|SetPlayerController _asfort_player_controller_athena)
    (Ability|ExecuteGameplayCueWithParamsReplicated (Variables|GameplayCues|GetGC_Pickup) (Ability|GameplayCue|MakeGameplayCueParameters :NormalizedMagnitude 0.0 :RawMagnitude 0.0 :Location (Transformation|GetActorLocation) :Normal (Transformation|GetActorForwardVector) :GameplayEffectLevel 1 :AbilityLevel (Math|Conversions|ToInteger(Byte) (Variables|GameplayCues|GetObjectWeights)) :bReplicateLocationWhenUsingMinimalRepProxy false))
    (if (Class|FortItemEntryComponent|HasItem (Variables|Default|GetFortItemEntry))
      (Class|FortItemEntryComponent|AddItemtoInventoryOwner (Variables|Default|GetFortItemEntry) (Variables|Default|GetPlayerController) false)
      (else
        (Inventory|GiveItemtoInventoryOwner (Variables|Default|GetPlayerController) (Variables|Default|GetWeaponItemDefinition))))
    (CallFunction|%28ServerOnly%29HideandDestroy)))

(event Curie|EventOnCurieElementInteract
  (params (CurieContainerHandle :type Curie%20Container%20Handle%20Structure) (ElementTag :type Gameplay%20Tag%20Structure) (InteractParamsHandle :type Curie%20Interact%20Params%20Handle%20Structure%20%28by%20ref%29))
  ;[1 1 1] On Curie Element Interact
  (|CallParentFunction CurieContainerHandle ElementTag InteractParamsHandle)
  (if (GameplayTags|Equal(GameplayTag) ElementTag "(TagName=\"Curie.Element.Fire\")")
    (CallFunction|SetFuseandDrop)))

(event EventOnDamageServer
  (params (Damage :type Float%20%28single-precision%29) (DamageTags :type Gameplay%20Tag%20Container%20Structure%20%28by%20ref%29) (Momentum :type Vector :default "0, 0, 0") (HitInfo :type Hit%20Result%20Structure%20%28by%20ref%29) (InstigatedBy :type Controller%20Object%20Reference) (DamageCauser :type Actor%20Object%20Reference) (EffectContext :type Gameplay%20Effect%20Context%20Handle%20Structure))
  (bind _returnvalue (Actor|GetOwner))
  (bind _explodeinstigator (Variables|Default|GetExplodeInstigator))
  ;[1 1 1] EVENT -> Damaged (Server)
  (Utilities|FlowControl|DoOnce false
    (:Completed
      (bind _asfort_player_controller (Utilities|Casting|CastToFortPlayerController InstigatedBy)
        (:then
          (Variables|Default|SetQuestController _asfort_player_controller)
          (Variables|Default|SetExplodeInstigator (Game|Player|GetPlayerPawn _asfort_player_controller)))
        (:CastFailed
          (bind _asfort_athena_aibot_controller (Utilities|Casting|CastToFortAthenaAIBotController InstigatedBy)
            (:then
              _)
            (:CastFailed
              _))
          (bind _asfort_pawn (Utilities|Casting|CastToFortPawn (Pawn|GetControlledPawn _asfort_athena_aibot_controller)))
          (Variables|Default|SetExplodeInstigator _asfort_pawn)))
      (Variables|Default|SetLastHit HitInfo) ; On Rep - calls On Fire GC
      (CallFunction|SetFuseandDrop)
      (if (and (Team|OnSameTeam _returnvalue _explodeinstigator) (> (Math|Vector|VectorGetAbsMax (Transformation|GetVelocity self)) 5.0))
        (bind _asfort_player_pawn (Utilities|Casting|CastToFortPlayerPawn _returnvalue))
        (if (not (Utilities|Equal(Object) _asfort_player_pawn _explodeinstigator))
          (bind _questcontroller (Variables|Default|GetQuestController)
            (:then
              (bind _returnvalue_2 (GameplayTags|GetOwnedGameplayTags DamageCauser :!class /Script/GameplayTags.GameplayTagAssetInterface))
              (Quest|SendComplexCustomStatEvent (Quests|GetQuestManager _questcontroller "Invalid") 0 _returnvalue_2 "(GameplayTags=((TagName=\"Athena.Quests.PetrolSkeet\")))" 1))
            (:else)))))))

(event Curie|EventOnCurieElementInteractBegun
  (params (CurieContainerHandle :type Curie%20Container%20Handle%20Structure) (ElementTag :type Gameplay%20Tag%20Structure) (InteractParamsHandle :type Curie%20Interact%20Params%20Handle%20Structure%20%28by%20ref%29))
  ;[1 1 1] On Curie Element Interact
  (|CallParentFunction CurieContainerHandle ElementTag InteractParamsHandle)
  (if (GameplayTags|Equal(GameplayTag) ElementTag "(TagName=\"Curie.Element.Fire\")")
    (CallFunction|SetFuseandDrop)))

(event Custom|ExplodeAndDie
  (bind _explodeinstigator (Variables|Default|GetExplodeInstigator))
  (bind _returnvalue (Utilities|IsValid _explodeinstigator :!class /Script/Engine.KismetSystemLibrary))
  (bind _self self)
  (bind _returnvalue_2 (select _returnvalue _explodeinstigator _self))
  (bind _returnvalue_3 (Ability|GetFortAbilitySystemComponent _returnvalue_2))
  ;[1 1 1] Explode
  (if (Variables|Default|GetShouldExplode)
    (if (Variables|Default|GetInWater)
      (bind _returnvalue_4 (Effects|Components|ParticleSystem|SpawnFXAtLocation "/Game/Effects/Explosion_Update/NS_GrenadeLauncher_Explosion_Asteria_Water.NS_GrenadeLauncher_Explosion_Asteria_Water" (Transformation|GetWorldLocation (Variables|Default|GetCapsule))))
      (Effects|Components|ParticleSystem|SetFloatParameter _returnvalue_4 "Global Scale" 0.65)
      (else
        (GameplayCue|ExecuteGameplayCue%28Burst%29Local self (Variables|GameplayCues|GetGC_Explode) (Ability|GameplayCue|MakeGameplayCueParameters 0.0))))
    (Utilities|FlowControl|DoOnce false
      (:Completed
        (bind _outactors (Collision|SphereOverlapActors (Transformation|GetActorLocation) (ScalableFloat|GetValueatLevel (Variables|Default|GetExplosionRadius)) (Utilities|Array|MakeArray "ObjectTypeQuery3" "ObjectTypeQuery1" "ObjectTypeQuery2" "ObjectTypeQuery5" "ObjectTypeQuery4")))
        (for _array_element _outactors
          (bind _in_line_of_sight (Class|MissionBlueprintFunctionLibrary|CheckBasicActorLineOfSight self "0, 0, 0" _array_element "0, 0, 0" true "TraceTypeQuery5" "None" 15.0 true))
          (if _in_line_of_sight
            (Utilities|Array|AddUnique (Variables|Default|GetTargetedActors) _array_element)))
        ;[1 1 1] Apply Damage
        (for _array_element_2 (Variables|Default|GetTargetedActors)
          (Utilities|IsValid _array_element_2
            (:Is%20Valid
              (Utilities|IsValid (Ability|GetFortAbilitySystemComponent _array_element_2)
                (:Is%20Valid
                  (Utilities|IsValid _returnvalue_3
                    (:Is%20Valid
                      (GameplayEffects|FortApplyGameplayEffectToTarget :self _returnvalue_3 :GameplayEffectClass (CallFunction|GetGameplayEffectClassforTarget _array_element_2) :TargetOfEffect (Ability|GetFortAbilitySystemComponent _array_element_2) :OptionalSourceObject self :OptionalEffectCauser self :OptionalDamageSource self :Level 0.0) ; Apply Damage
                      ;[1 1 1] Do Curie Fire Application
                      (if (Curie|IsCurieActive)
                        (if (ScalableFloat|GetValueasBool (Variables|Default|GetExplosionCurieFireApplication))
                          (Curie|ExecuteFireApplication :Instigator (Variables|Default|GetExplodeInstigator) :EffectCauser self :Target _array_element_2 :Magnitude 1e+02 :PropagationFuel (ScalableFloat|GetValueasInteger (Variables|Default|GetExplosionCurieFirePropagationFuel)) :LandscapeIgnitionRadius 0.0)
                          (else
                            _))
                        ;[0.1667 0.1667 0.1667] Impulse Vehicles
                        (bind _asfort_athena_vehicle (Utilities|Casting|CastToFortAthenaVehicle _array_element_2))
                        (Audio|BroadcastAbilityImpactatLocation 0 _asfort_athena_vehicle (CallFunction|CalcVehicleImpulse _asfort_athena_vehicle (ScalableFloat|GetValueatLevel (Variables|Default|GetHotfixExplodeVehicleImpulse))) 777.0)
                        (Audio|BroadcastAbilitySpinThrustatLocation 0 _asfort_athena_vehicle (Math|Vector|MakeVector 0.0 (Math|Random|RandomFloatinRange (Math|Float|Min(Float) (* (ScalableFloat|GetValueatLevel (Variables|Default|GetHotfixExplodeSpinThrust)) -1.0) -5e+01) (Math|Float|Max(Float) (ScalableFloat|GetValueatLevel (Variables|Default|GetHotfixExplodeSpinThrust)) 5e+01)) (Math|Random|RandomFloatinRange (Math|Float|Min(Float) (* (ScalableFloat|GetValueatLevel (Variables|Default|GetHotfixExplodeSpinThrust)) -1.0) -5e+01) (Math|Float|Max(Float) (ScalableFloat|GetValueatLevel (Variables|Default|GetHotfixExplodeSpinThrust)) 5e+01))) 777.0)))))))))
        (CallFunction|HideandDestroy)))
    (else
      (CallFunction|HideandDestroy))))

(event Custom|SpawnCurieFireballs
  (params (Explodeinstigator :type Fort%20Pawn%20Object%20Reference))
  ;[1 1 1] (If active) Spawn Curie Actors.  Leave's a burning patch, and spawns the specified number of fireballs to produce more burning patches.
  (if (Curie|IsCurieActive)
    (Building|SpawnBuildingGameplayActor (Variables|Default|GetCurieBGAtoSpawn) (Math|Transform|MakeTransform (Transformation|GetActorLocation)) Explodeinstigator)
    (for _ (range (Math|Float|Truncate (ScalableFloat|GetValueatLevel (Variables|Default|GetFireBallsToSpawn))))
      (Projectile|SpawnProjectile :ProjectileClass (Variables|Default|GetCurieFireballsToSpawn) :RequestedBy Explodeinstigator :SpawnLocation (+ (Transformation|GetActorLocation) "0, 0,15.000000") :SpawnRotation "0, 0, 0" :SpawnDirection (Math|Rotator|MakeRotfromX (Math|Random|RandomUnitVectorinConeinDegrees (+ (Transformation|GetActorUpVector) (Utilities|RandomVectorinRange "-0.500000,-0.620000,-0.700000" "0.500000,0.620000,0.700000")) 2e+01)) :bUseDefaultPhysics false :InitialSpeed (Math|Random|RandomFloatinRange (ScalableFloat|GetValueatLevel (Variables|Default|GetSpawnedProjectileMinSpeed)) (ScalableFloat|GetValueatLevel (Variables|Default|GetSpawnedProjectileMaxSpeed))) :GravityScale 0.8 :ChargePercent 0.0))))

(event EventOnDeathServer
  (params (Damage :type Float%20%28single-precision%29) (DamageTags :type Gameplay%20Tag%20Container%20Structure%20%28by%20ref%29) (Momentum :type Vector :default "0, 0, 0") (HitInfo :type Hit%20Result%20Structure%20%28by%20ref%29) (InstigatedBy :type Controller%20Object%20Reference) (DamageCauser :type Actor%20Object%20Reference) (EffectContext :type Gameplay%20Effect%20Context%20Handle%20Structure))
  ;[1 1 1] Override on server death function to get rid of parent death fx
  (Rendering|SetActorHiddenInGame true)
  (Utilities|IsValid (Variables|Default|GetWeaponAttachedTo)
    (:Is%20Valid
      (bind _asb_petrol_weapon (Utilities|Casting|CastToB_PetrolWeapon (Variables|Default|GetWeaponAttachedTo)))
      (Class|BPetrolWeapon|BGADestroyedWhileHeld _asb_petrol_weapon)
      (Development|PrintString "Died" true true "(R=0.057843,G=1.000000,B=0.000000,A=1.000000)" 5.0))))

(event Custom|SetFuseAndDrop
  ;[1 1 1] Set Fuse and Drop
  (Variables|Default|SetFuseLit true)
  (Networking|SwitchHasAuthority
    (:Authority
      (bind _returnvalue (Utilities|Time|SetTimerbyEvent (EventDispatchers|CreateEvent :!event Custom|SpawnFireBallsExplodeAndDie) (Math|Random|RandomFloatinRange (ScalableFloat|GetValueatLevel (Variables|Default|GetFuseMinDuration)) (ScalableFloat|GetValueatLevel (Variables|Default|GetFuseMaxDuration)))))
      (Variables|Default|SetFuseTimerHandle _returnvalue)
      (if (Variables|Default|GetIsAttachedToWeapon)
        (CallFunction|DropHeldItem)))))

(event Custom|DropHeldItem
  ;[1 1 1] Drop Held Item
  (Networking|FlushNetDormancy)
  (Variables|Default|SetDropped true)
  (Transformation|DetachFromActor self "KeepWorld" "KeepWorld" "KeepWorld")
  (Development|SetHiddeninGame (Variables|Default|GetPetrolCan) false false)
  (Collision|SetCollisionProfileName (Variables|Default|GetCapsule) "FortProjectileHitAllPawns" true :!class /Script/Engine.PrimitiveComponent)
  (Collision|SetCollisionEnabled (Variables|Default|GetPetrolCan) "QueryAndPhysics")
  (Projectile|SetVelocity (Variables|Default|GetFortProjectileMovement) (0.0 0.0 1.0))
  (bind _asb_petrol_weapon (Utilities|Casting|CastToB_PetrolWeapon (Variables|Default|GetWeaponAttachedTo))
    (:then
      (Class|BPetrolWeapon|BGADestroyedWhileHeld _asb_petrol_weapon))
    (:CastFailed
      _))
  (InventoryManagement|RemoveItem (Variables|Default|GetHeldItemInventoryInterface) (Variables|Default|GetHeldItemGUID) -1))

(event Custom|SpawnFireBallsExplodeAndDie
  ;[1 1 1] Spawn Fireballs, Explode, and Die
  (Networking|SwitchHasAuthority
    (:Authority
      (CallFunction|SpawnCurieFireballs (Variables|Default|GetExplodeInstigator))
      (CallFunction|ExplodeandDie))))

(event EventDestroyed
  ;[1 1 1] Destroyed
  (Development|PrintString "Petrol Pickup BGA Destroyed" true true "(R=1.000000,G=0.000000,B=0.021146,A=1.000000)" 5.0))

(event Custom|SetHeldItemInfo
  (params (InventoryInterface :type Fort%20Inventory%20Owner%20Interface%20Interface) (GUID :type Guid%20Structure) (WeaponAttachedTo :type Fort%20Weapon%20Object%20Reference))
  ;[1 1 1] Set Held Item Info
  (Variables|Default|SetHeldItemInventoryInterface InventoryInterface)
  (Variables|Default|SetHeldItemGUID GUID)
  (Variables|Default|SetWeaponAttachedTo WeaponAttachedTo)
  (Variables|Default|SetIsAttachedToWeapon true))

(event Custom|HideAndDestroy
  ;[1 1 1] Hide and Destroy
  (Development|SetHiddeninGame (Variables|Default|GetPetrolCan) true false)
  (Actor|DestroyActor))

(event Custom|%28ServerOnly%29HideAndDestroy
  ;[1 1 1] Hide and Destroy
  (Development|SetHiddeninGame (Variables|Default|GetPetrolCan) true false)
  (Actor|DestroyActor))

(event Custom|Refuel
  (params (FuelToAdd :type Integer) (InstigatingController :type Fort%20Player%20Controller%20Object%20Reference) (InstigatingActor :type Actor%20Object%20Reference))
  ;[1 1 1] Refuel
  (CallFunction|RefuelFunction FuelToAdd InstigatingController InstigatingActor))

(event OnProjectileBounce(FortProjectileMovement)
  (params (ImpactResult :type Hit%20Result%20Structure%20%28by%20ref%29) (ImpactVelocity :type Vector%20%28by%20ref%29 :default "0, 0, 0"))
  (bind (_bblockinghit _binitialoverlap _time _distance _location _impactpoint _normal _impactnormal _physmat _hitactor _hitcomponent) (Collision|BreakHitResult ImpactResult))
  (bind _returnvalue (Components|ComponentHasTag _hitcomponent "NoStick"))
  (bind _returnvalue_2 (Actor|ActorHasTag _hitactor "NoStick"))
  (bind _returnvalue_3 (or _returnvalue_2 _returnvalue))
  (bind (_asfort_vehicle_interface _bsuccess) (Utilities|Casting|CastToFortVehicleInterface _hitactor))
  ;[1 1 1] On bounce, if we hit a no stick object, adjust velocity to bounce away
  (if _returnvalue_3
    _
    (elif (and (not _returnvalue_3) _bsuccess)
      ;;[1 0.6317 0.6816] JIRA - FORT-457573: BGA gets stuck in/around vehicles when hitting supporting structure at higher speeds. This check reduces chances of that while allowing fun gameplay of storing gas cans in truck beds etc.
      (bind _asfort_vehicle_interface_2 (Utilities|Casting|CastToFortVehicleInterface _asfort_vehicle_interface))
      (if (> (Vehicle|GetForwardSpeedKmh _asfort_vehicle_interface_2) 3e+01)
        _)))
  (Projectile|SetVelocity (Variables|Default|GetFortProjectileMovement) (* (Math|Vector|Normalize (Math|Vector|vector+vector (Math|Vector|MirrorVectorbyNormal (* (Variables|/Script/Engine.MovementComponent|GetVelocity (Variables|Default|GetFortProjectileMovement)) 0.75) "0.000000, 0,1.000000") (Math|Vector|MakeVector 0.0 0.0 (* (Math|Vector|VectorLength (Variables|/Script/Engine.MovementComponent|GetVelocity (Variables|Default|GetFortProjectileMovement))) 0.25)))) (Math|Float|Max(Float) (Math|Vector|VectorLength (Math|Vector|vector+vector (Math|Vector|MirrorVectorbyNormal (* (Variables|/Script/Engine.MovementComponent|GetVelocity (Variables|Default|GetFortProjectileMovement)) 0.75) "0.000000, 0,1.000000") (Math|Vector|MakeVector 0.0 0.0 (* (Math|Vector|VectorLength (Variables|/Script/Engine.MovementComponent|GetVelocity (Variables|Default|GetFortProjectileMovement))) 0.25)))) (Variables|Default|GetMinSpeedOfNoStickBounce)))))

(event OnEnterWaterEvent(FortWaterInteraction)
  (params (WaterBody :type Fort%20Water%20Body%20Actor%20Object%20Reference) (WaterInteractionComponent :type Fort%20Water%20Interaction%20Component%20Object%20Reference) (bIsFirstBody :type Boolean))
  (bind _ns_object_floatinginwater (Variables|Default|GetNS_Object_FloatingInWater))
  ;[1 1 1] Get water info
  (Variables|Default|SetInWater true)
  (bind _outwatersurfaceinfo (WaterBody|GetWaterSurfaceInfoStruct WaterBody (Transformation|GetActorLocation) true false false))
  (Components|Activation|Activate _ns_object_floatinginwater false)
  (Effects|Components|ParticleSystem|SetFloatParameter _ns_object_floatinginwater "User.WaterDepth" (.WaterDepth (Utilities|Struct|BreakWaterSurfaceInfo _outwatersurfaceinfo)))
  (Effects|Components|ParticleSystem|SetVectorParameter _ns_object_floatinginwater "User.WaterPlaneLocation" (Utilities|Struct|BreakWaterSurfaceInfo _outwatersurfaceinfo))
  (Effects|Components|ParticleSystem|SetVectorParameter _ns_object_floatinginwater "User.WaterPlaneNormal" (.WaterPlaneNormal (Utilities|Struct|BreakWaterSurfaceInfo _outwatersurfaceinfo))))

(event OnExitWaterEvent(FortWaterInteraction)
  (params (WaterBody :type Fort%20Water%20Body%20Actor%20Object%20Reference) (WaterInteractionComponent :type Fort%20Water%20Interaction%20Component%20Object%20Reference) (bIsLastBody :type Boolean))
  (Variables|Default|SetInWater false)
  (Components|Activation|Deactivate (Variables|Default|GetNS_Object_FloatingInWater)))

(fn BlueprintCanInteract
  (params (InteractingPawn :type Fort%20Pawn%20Object%20Reference) (InteractionBeingAttempted :type EInteractionBeingAttempted%20Enum :default "FirstInteraction") (InteractionType :type TInteractionType%20Enum :default "IT_NoInteraction"))
  (return (and (not (Variables|Default|GetFuseLit)) (not (GameplayTags|HasAnyMatchingGameplayTags InteractingPawn (Variables|Interact|GetTC_BlockPickup))))))

(fn BlueprintGetInteractionString
  (params (InteractingPawn :type Fort%20Pawn%20Object%20Reference) (InteractionBeingAttempted :type EInteractionBeingAttempted%20Enum :default "FirstInteraction"))
  (return "NSLOCTEXT(\"[41AA2EF147203A0C8CCFC68E90A212AB]\", \"7D8614864DB0F05503DB8B9B511D75CF\", \"Pickup Gas Can\")"))

(fn BlueprintGetInteractionTime
  (params (InteractingPawn :type Fort%20Pawn%20Object%20Reference) (InteractionBeingAttempted :type EInteractionBeingAttempted%20Enum :default "FirstInteraction"))
  (return 0.25 false))

(fn BlueprintOnLocalInteract
  (params (InteractingPawn :type Fort%20Player%20Pawn%20Object%20Reference) (InteractionBeingAttempted :type EInteractionBeingAttempted%20Enum :default "FirstInteraction"))
  ;[1 1 1] Prevent player from attempting to pick up Gas Can if inventory is full (and player is holding pickaxe so the pickup won't replace another quickbar slotted item)
  (if (> (Inventory|CountInventoryOverflowfromAddingItem (Fortnite|GetFortPlayerControllerfromActor InteractingPawn) (Loot|CreateItemEntry (Variables|Default|GetWeaponItemDefinition)) false true) 0)
    (FailedAction|UnabletoPerformAction InteractingPawn true (Variables|Default|GetTag_FailedInteract_InventoryIsFull))
    (Audio|PlaySoundatLocation (Variables|Default|GetFailedInteractSound) (Transformation|GetActorLocation))
    (return true)
    (else
      (return false))))

(fn CalcVehicleImpulse
  (params (Vehicle :type Actor%20Object%20Reference) (Magnitude :type Float%20%28double-precision%29))
  (return (Math|Vector|MakeVector (.x (* (Math|Vector|Normalize (- (Transformation|GetActorLocation Vehicle) (Transformation|GetActorLocation))) Magnitude)) (.y (* (Math|Vector|Normalize (- (Transformation|GetActorLocation Vehicle) (Transformation|GetActorLocation))) Magnitude)) (Math|Float|Max(Float) (+ (.z (* (Math|Vector|Normalize (- (Transformation|GetActorLocation Vehicle) (Transformation|GetActorLocation))) Magnitude)) 1.0) 2.5e+02))))

(fn GetGameplayEffectClassForTarget
  (params (Target :type Actor%20Object%20Reference))
  (Utilities|Casting|CastToFortPawn Target
    (:then
      (return (Variables|Default|GetGE_ExplodePlayerDamage)))
    (:CastFailed
      (Utilities|Casting|CastToFortAthenaVehicle Target
        (:then
          (return (Variables|Default|GetGE_ExplodeVehicleDamage)))
        (:CastFailed
          (return (Variables|Default|GetGE_ExplodeBuildingDamage)))))))

(fn OnRep_LastHit
  (bind _self self)
  (Networking|CanExecuteCosmeticEvents
    (:True
      (bind _outcueparameters (GameplayCue|BuildCueParametersfromHitResult _self (Variables|Default|GetLastHit)))
      (Projectile|AddGameplayCue%28Looping%29Local _self (Variables|GameplayCues|GetGC_OnFire) _outcueparameters))))

(fn OnRep_ShouldExplode)

(fn QuestManagerRefuelStatEvent
  (params (FortPlayerController :type Fort%20Player%20Controller%20Object%20Reference) (RefuelAmount :type Integer) (AdditionalSourceTags :type Gameplay%20Tag%20Container%20Structure))
  (GameplayTags|AppendGameplayTagContainers AdditionalSourceTags (Tags|GetWeaponDataGameplayTags (Weapon|GetCurrentWeapon (Game|Player|GetPlayerPawn FortPlayerController))))
  (Quest|SendCustomStatEventwithTags :self (Quests|GetQuestManager FortPlayerController "Athena") :Type "Refuel" :AdditionalSourceTags AdditionalSourceTags :Count RefuelAmount))

(fn Refuel_Function
  (params (FuelToAdd :type Integer) (InstigatingController :type Fort%20Player%20Controller%20Object%20Reference) (InstigatingActor :type Actor%20Object%20Reference))
  (bind _fortitementry (Variables|Default|GetFortItemEntry))
  (bind _returnvalue (Class|FortItemEntryComponent|GetLoadedAmmo _fortitementry))
  (bind _weaponattachedto (Variables|Default|GetWeaponAttachedTo))
  (Networking|SwitchHasAuthority
    (:Authority
      ;[1 1 1] For quest progression on this side, I can't figure out what this is supposed to do, but I don't seem to be hitting any breakpoints, so leaving this as is.
      (if (Variables|Default|GetIsAttachedToWeapon)
        (Utilities|IsValid _weaponattachedto
          (:Is%20Valid
            (FortWeapon|ModAmmo _weaponattachedto (Variables|GetFuelToAdd))
            (Development|PrintString (Utilities|String|Append "Fuel added to attached weapon = " (Utilities|String|ToString(Integer) (Variables|GetFuelToAdd))) true true "(R=0.000000,G=1.000000,B=0.102666,A=1.000000)")))
        (else
          (Utilities|IsValid (Variables|Default|GetFortItemEntry)
            (:Is%20Valid
              (LocalVariables|SetAmmoBefore _returnvalue)
              (bind _returnvalue_2 (Class|FortItemEntryComponent|TrySetLoadedAmmo (Variables|Default|GetFortItemEntry) (+ _returnvalue (Variables|GetFuelToAdd))))
              (bind _output_get (LocalVariables|SetAmmoAdded (- _returnvalue_2 (LocalVariables|GetAmmoBefore))))
              (Development|PrintString (Utilities|String|Append "Fuel added to detached BGA = " (Utilities|String|ToString(Integer) _output_get)) true true "(R=0.003054,G=1.000000,B=0.000000,A=1.000000)")
              (Utilities|FlowControl|Sequence
                (:then_0
                  _)
                (:then_1
                  (if (> (LocalVariables|GetAmmoAdded) 0)
                    (bind _returnvalue_3 (GameplayTags|GetOwnedGameplayTags (Variables|GetInstigatingActor) :!class /Script/GameplayTags.GameplayTagAssetInterface))
                    (CallFunction|QuestManagerRefuelStatEvent (Variables|GetInstigatingController) (LocalVariables|GetAmmoAdded) _returnvalue_3))))))))
      (Ability|ExecuteGameplayCueWithParamsReplicated (Variables|GameplayCues|GetGC_Refuel) (Ability|GameplayCue|MakeGameplayCueParameters :NormalizedMagnitude 0.0 :RawMagnitude 0.0 :Location (Transformation|GetActorLocation) :Normal (Transformation|GetActorForwardVector) :GameplayEffectLevel 1 :AbilityLevel (Math|Conversions|ToInteger(Byte) (Variables|GameplayCues|GetObjectWeights)) :bReplicateLocationWhenUsingMinimalRepProxy false)))))

(fn ShouldDie
  (params (Damage :type Float%20%28single-precision%29) (EventInstigator :type Controller%20Object%20Reference) (DamageCauser :type Actor%20Object%20Reference) (bIsTest :type Boolean) (EffectContext :type Gameplay%20Effect%20Context%20Handle%20Structure%20%28by%20ref%29))
  ;[1 1 1] This is set to false so that we don't lose collision when this is shot out of a player's hand - call Hide and Destroy to finish killing the object
  (return false))

(fn ConstructionScript)
